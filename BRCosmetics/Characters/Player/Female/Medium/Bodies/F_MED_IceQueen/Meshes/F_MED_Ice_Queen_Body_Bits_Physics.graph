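# Physics asset for the female Ice Queen cosmetic body bits, covering dangling fur, arm cuffs and leg cloth bones.
# Small capsules on clavicle and arm bones anchor the fur and tightly limited cuffs, while each leg cloth forms a separate capsule-to-box chain.
# Left and right differ: left fur joints are limited but right fur joints swing and twist freely, and elbow cuff B exists only on the left.
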PhysicsAsset /BRCosmetics/Characters/Player/Female/Medium/Bodies/F_MED_IceQueen/Meshes/F_MED_Ice_Queen_Body_Bits_Physics
bodies {
 clavicle_l [{shapeType=Capsule, length=1, radius=1}]
 clavicle_r [{shapeType=Capsule, length=1, radius=1, rotation=<rot 0 180 0>}]
 dyn_elbow_cuff_A_01_l [{shapeType=Capsule, length=2, radius=2, center=<2 0 0>, rotation=<rot 90 0 0>}]
 dyn_elbow_cuff_A_01_r [{shapeType=Capsule, length=2, radius=2, center=<2 0 0>, rotation=<rot 90 0 0>}]
 dyn_elbow_cuff_B_01_l [{shapeType=Capsule, length=2, radius=2, center=<2 0 0>, rotation=<rot 90 0 0>}]
 dyn_fur_01_l [{shapeType=Capsule, length=1, radius=4, center=<2 0 0>, rotation=<rot 90 0 0>}]
 dyn_fur_01_r [{shapeType=Capsule, length=1, radius=4, center=<2 0 0>, rotation=<rot 90 0 0>}]
 dyn_fur_02_l [{shapeType=Capsule, length=1, radius=4, center=<2 0 0>, rotation=<rot 90 0 0>}]
 dyn_fur_02_r [{shapeType=Capsule, length=1, radius=4, center=<2 0 0>, rotation=<rot 90 0 0>}]
 dyn_fur_03_l [{shapeType=Capsule, length=1, radius=4, center=<2 0 0>, rotation=<rot 90 0 0>}]
 dyn_fur_03_r [{shapeType=Capsule, length=1, radius=4, center=<2 0 0>, rotation=<rot 90 0 0>}]
 dyn_fur_04_l [{shapeType=Capsule, length=1, radius=4, center=<2 0 0>, rotation=<rot 90 0 0>}]
 dyn_fur_04_r [{shapeType=Capsule, length=1, radius=4, center=<2 0 0>, rotation=<rot 90 0 0>}]
 dyn_leg_cloth_03_l [{shapeType=Capsule, length=0, radius=2}]
 dyn_leg_cloth_03_r [{shapeType=Capsule, length=0, radius=2}]
 dyn_leg_cloth_04_l [{shapeType=Box, extentX=15, extentY=5, extentZ=15, center=<7 0 0>}]
 dyn_leg_cloth_04_r [{shapeType=Box, extentX=15, extentY=5, extentZ=15, center=<7 0 0>}]
 dyn_leg_cloth_05_l [{shapeType=Box, extentX=15, extentY=5, extentZ=15, center=<7 0 0>}]
 dyn_leg_cloth_05_r [{shapeType=Box, extentX=15, extentY=5, extentZ=15, center=<7 0 0>}]
 dyn_upperarm_cuff_A_01_l [{shapeType=Capsule, length=2, radius=2, center=<2 0 0>, rotation=<rot 90 0 0>}]
 dyn_upperarm_cuff_A_01_r [{shapeType=Capsule, length=2, radius=2, center=<2 0 0>, rotation=<rot 90 0 0>}]
 dyn_upperarm_cuff_B_01_l [{shapeType=Capsule, length=2, radius=2, center=<2 0 0>, rotation=<rot 90 0 0>}]
 dyn_upperarm_cuff_B_01_r [{shapeType=Capsule, length=2, radius=2, center=<2 0 0>, rotation=<rot 90 0 0>}]
 dyn_upperarm_cuff_B_02_l [{shapeType=Capsule, length=2, radius=2, center=<2 0 0>, rotation=<rot 90 0 0>}]
 dyn_upperarm_cuff_B_02_r [{shapeType=Capsule, length=2, radius=2, center=<2 0 0>, rotation=<rot 90 0 0>}]
 lowerarm_l [{shapeType=Capsule, length=1, radius=1}]
 lowerarm_r [{shapeType=Capsule, length=1, radius=1, rotation=<rot 0 180 0>}]
 upperarm_l [{shapeType=Capsule, length=1, radius=1}]
 upperarm_r [{shapeType=Capsule, length=1, radius=1, rotation=<rot 0 180 0>}]
 upperarm_twist_02_l [{shapeType=Capsule, length=1, radius=1}]
 upperarm_twist_02_r [{shapeType=Capsule, length=1, radius=1, rotation=<rot 0 180 0>}]
}
constraints [
 {bone1Name=dyn_leg_cloth_04_l, bone2Name=dyn_leg_cloth_03_l, swing1LimitDegrees=45, swing2LimitDegrees=45, twistLimitDegrees=5}
 {bone1Name=dyn_leg_cloth_05_l, bone2Name=dyn_leg_cloth_04_l, swing1LimitDegrees=45, swing2LimitDegrees=45, twistLimitDegrees=5}
 {bone1Name=dyn_leg_cloth_04_r, bone2Name=dyn_leg_cloth_03_r, swing1LimitDegrees=45, swing2LimitDegrees=45, twistLimitDegrees=5}
 {bone1Name=dyn_leg_cloth_05_r, bone2Name=dyn_leg_cloth_04_r, swing1LimitDegrees=45, swing2LimitDegrees=45, twistLimitDegrees=5}
 {bone1Name=dyn_fur_03_l, bone2Name=clavicle_l, swing1LimitDegrees=45, swing2LimitDegrees=45, twistLimitDegrees=5}
 {bone1Name=dyn_fur_04_l, bone2Name=clavicle_l, swing1LimitDegrees=45, swing2LimitDegrees=45, twistLimitDegrees=5}
 {bone1Name=dyn_fur_02_l, bone2Name=clavicle_l, swing1LimitDegrees=45, swing2LimitDegrees=45, twistLimitDegrees=5}
 {bone1Name=dyn_fur_01_l, bone2Name=clavicle_l, swing1LimitDegrees=45, swing2LimitDegrees=45, twistLimitDegrees=5}
 {bone1Name=dyn_elbow_cuff_B_01_l, bone2Name=lowerarm_l, swing1LimitDegrees=5, swing2LimitDegrees=20, twistLimitDegrees=5}
 {bone1Name=dyn_upperarm_cuff_B_01_l, bone2Name=upperarm_twist_02_l, swing1LimitDegrees=10, swing2LimitDegrees=20, twistLimitDegrees=5}
 {bone1Name=dyn_upperarm_cuff_B_02_l, bone2Name=dyn_upperarm_cuff_B_01_l, swing1LimitDegrees=5, swing2LimitDegrees=20, twistLimitDegrees=5}
 {bone1Name=dyn_upperarm_cuff_A_01_l, bone2Name=upperarm_twist_02_l, swing1LimitDegrees=5, swing2LimitDegrees=20, twistLimitDegrees=5}
 {bone1Name=dyn_elbow_cuff_A_01_l, bone2Name=upperarm_l, swing1LimitDegrees=5, swing2LimitDegrees=20, twistLimitDegrees=5}
 {
  bone1Name=lowerarm_l
  bone2Name=upperarm_l
  swing1LimitDegrees=45
  swing1Motion=Free
  swing2LimitDegrees=45
  swing2Motion=Free
  twistLimitDegrees=45
  twistMotion=Free
 }
 {
  bone1Name=upperarm_l
  bone2Name=clavicle_l
  swing1LimitDegrees=45
  swing1Motion=Free
  swing2LimitDegrees=45
  swing2Motion=Free
  twistLimitDegrees=45
  twistMotion=Free
 }
 {
  bone1Name=upperarm_twist_02_l
  bone2Name=upperarm_l
  swing1LimitDegrees=45
  swing1Motion=Free
  swing2LimitDegrees=45
  swing2Motion=Free
  twistLimitDegrees=45
  twistMotion=Free
 }
 {
  bone1Name=lowerarm_r
  bone2Name=upperarm_r
  swing1LimitDegrees=45
  swing1Motion=Free
  swing2LimitDegrees=45
  swing2Motion=Free
  twistLimitDegrees=45
  twistMotion=Free
 }
 {
  bone1Name=upperarm_twist_02_r
  bone2Name=upperarm_r
  swing1LimitDegrees=45
  swing1Motion=Free
  swing2LimitDegrees=45
  swing2Motion=Free
  twistLimitDegrees=45
  twistMotion=Free
 }
 {
  bone1Name=upperarm_r
  bone2Name=clavicle_r
  swing1LimitDegrees=45
  swing1Motion=Free
  swing2LimitDegrees=45
  swing2Motion=Free
  twistLimitDegrees=45
  twistMotion=Free
 }
 {
  bone1Name=dyn_fur_03_r
  bone2Name=clavicle_r
  swing1LimitDegrees=45
  swing1Motion=Free
  swing2LimitDegrees=45
  swing2Motion=Free
  twistLimitDegrees=45
  twistMotion=Free
 }
 {
  bone1Name=dyn_fur_04_r
  bone2Name=clavicle_r
  swing1LimitDegrees=45
  swing1Motion=Free
  swing2LimitDegrees=45
  swing2Motion=Free
  twistLimitDegrees=45
  twistMotion=Free
 }
 {
  bone1Name=dyn_fur_02_r
  bone2Name=clavicle_r
  swing1LimitDegrees=45
  swing1Motion=Free
  swing2LimitDegrees=45
  swing2Motion=Free
  twistLimitDegrees=45
  twistMotion=Free
 }
 {
  bone1Name=dyn_fur_01_r
  bone2Name=clavicle_r
  swing1LimitDegrees=45
  swing1Motion=Free
  swing2LimitDegrees=45
  swing2Motion=Free
  twistLimitDegrees=45
  twistMotion=Free
 }
 {bone1Name=dyn_elbow_cuff_A_01_r, bone2Name=upperarm_r, swing1LimitDegrees=5, swing2LimitDegrees=20, twistLimitDegrees=5}
 {bone1Name=dyn_upperarm_cuff_B_02_r, bone2Name=dyn_upperarm_cuff_B_01_r, swing1LimitDegrees=5, swing2LimitDegrees=20, twistLimitDegrees=5}
 {bone1Name=dyn_upperarm_cuff_B_01_r, bone2Name=upperarm_twist_02_r, swing1LimitDegrees=5, swing2LimitDegrees=20, twistLimitDegrees=5}
 {bone1Name=dyn_upperarm_cuff_A_01_r, bone2Name=upperarm_twist_02_r, swing1LimitDegrees=5, swing2LimitDegrees=20, twistLimitDegrees=5}
]
