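# Physics asset for the female medium Ice Queen cosmetic body, providing leg collision and a jointed skirt rig.
# Capsules cover pelvis, thighs, calves and feet, while thin boxes form skirt bone chains hanging from the pelvis with limited swing and zero twist.
# Angularly free cross-constraints tie neighboring chains together, mostly lower down, linking the panels.
# The right front and side chains are shorter than the left, ending in enlarged boxes.
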PhysicsAsset /BRCosmetics/Characters/Player/Female/Medium/Bodies/F_MED_IceQueen/Meshes/F_MED_Ice_Queen_Body_Physics
bodies {
 calf_l [{shapeType=Capsule, length=33, radius=11.5, center=<-23.9639 -0.0004 0>, rotation=<rot 90 0 0>}]
 calf_r [{shapeType=Capsule, length=36.25, radius=11.5, center=<24.1407 0 0>, rotation=<rot 90 0 0>}]
 dyn_skirt_back_02 [{
   shapeType=Box
   extentX=15
   extentY=2
   extentZ=15
   shapeName="None, None, None, None, None, None, None, None, None, None, None, None, None, None, None, None, None, None, None, None, None, None, None, None, None, None, None, None, None, None, None, None, None, None"
   center=<7.5 0 0>
   rotation=<rot 90 0 0>
  }]
 dyn_skirt_back_03 [{
   shapeType=Box
   extentX=15
   extentY=2
   extentZ=15
   shapeName="None, None, None, None, None, None, None, None, None, None, None, None, None, None, None, None, None, None, None, None, None, None, None, None, None, None, None, None, None, None, None, None, None, None"
   center=<7.5 0 0>
   rotation=<rot 90 0 0>
  }]
 dyn_skirt_back_04 [{
   shapeType=Box
   extentX=15
   extentY=2
   extentZ=15
   shapeName="None, None, None, None, None, None, None, None, None, None, None, None, None, None, None, None, None, None, None, None, None, None, None, None, None, None, None, None, None, None, None, None, None, None"
   center=<7.5 0 0>
   rotation=<rot 90 0 0>
  }]
 dyn_skirt_back_05 [{
   shapeType=Box
   extentX=15
   extentY=2
   extentZ=15
   shapeName="None, None, None, None, None, None, None, None, None, None, None, None, None, None, None, None, None, None, None, None, None, None, None, None, None, None, None, None, None, None, None, None, None, None"
   center=<7.5 0 0>
   rotation=<rot 90 0 0>
  }]
 dyn_skirt_back_06 [{
   shapeType=Box
   extentX=15
   extentY=2
   extentZ=15
   shapeName="None, None, None, None, None, None, None, None, None, None, None, None, None, None, None, None, None, None, None, None, None, None, None, None, None, None, None, None, None, None, None, None, None, None"
   center=<7.5 0 0>
   rotation=<rot 90 0 0>
  }]
 dyn_skirt_back_07 [{
   shapeType=Box
   extentX=15
   extentY=2
   extentZ=15
   shapeName="None, None, None, None, None, None, None, None, None, None, None, None, None, None, None, None, None, None, None, None, None, None, None, None, None, None, None, None, None, None, None, None, None, None"
   center=<7.5 0 0>
   rotation=<rot 90 0 0>
  }]
 dyn_skirt_mid_02_l [{
   shapeType=Box
   extentX=15
   extentY=2
   extentZ=15
   shapeName="None, None, None, None, None, None, None, None, None, None, None, None, None, None, None, None, None, None, None, None, None, None, None, None, None, None, None, None, None, None, None, None, None, None"
   center=<7.5 0 0>
   rotation=<rot 90 0 0>
  }]
 dyn_skirt_mid_02_r [{
   shapeType=Box
   extentX=15
   extentY=2
   extentZ=15
   shapeName="None, None, None, None, None, None, None, None, None, None, None, None, None, None, None, None, None, None, None, None, None, None, None, None, None, None, None, None, None, None, None, None, None, None"
   center=<7.5 0 0>
   rotation=<rot 90 0 0>
  }]
 dyn_skirt_mid_03_l [{
   shapeType=Box
   extentX=15
   extentY=2
   extentZ=15
   shapeName="None, None, None, None, None, None, None, None, None, None, None, None, None, None, None, None, None, None, None, None, None, None, None, None, None, None, None, None, None, None, None, None, None, None"
   center=<7.5 0 0>
   rotation=<rot 90 0 0>
  }]
 dyn_skirt_mid_03_r [{
   shapeType=Box
   extentX=15
   extentY=2
   extentZ=15
   shapeName="None, None, None, None, None, None, None, None, None, None, None, None, None, None, None, None, None, None, None, None, None, None, None, None, None, None, None, None, None, None, None, None, None, None"
   center=<7.5 0 0>
   rotation=<rot 90 0 0>
  }]
 dyn_skirt_mid_04_l [{
   shapeType=Box
   extentX=15
   extentY=2
   extentZ=15
   shapeName="None, None, None, None, None, None, None, None, None, None, None, None, None, None, None, None, None, None, None, None, None, None, None, None, None, None, None, None, None, None, None, None, None, None"
   center=<7.5 0 0>
   rotation=<rot 90 0 0>
  }]
 dyn_skirt_mid_04_r [{
   shapeType=Box
   extentX=15
   extentY=2
   extentZ=15
   shapeName="None, None, None, None, None, None, None, None, None, None, None, None, None, None, None, None, None, None, None, None, None, None, None, None, None, None, None, None, None, None, None, None, None, None"
   center=<7.5 0 0>
   rotation=<rot 90 0 0>
  }]
 dyn_skirt_mid_05_l [{
   shapeType=Box
   extentX=15
   extentY=2
   extentZ=15
   shapeName="None, None, None, None, None, None, None, None, None, None, None, None, None, None, None, None, None, None, None, None, None, None, None, None, None, None, None, None, None, None, None, None, None, None"
   center=<7.5 0 0>
   rotation=<rot 90 0 0>
  }]
 dyn_skirt_mid_05_r [{
   shapeType=Box
   extentX=15
   extentY=2
   extentZ=15
   shapeName="None, None, None, None, None, None, None, None, None, None, None, None, None, None, None, None, None, None, None, None, None, None, None, None, None, None, None, None, None, None, None, None, None, None"
   center=<7.5 0 0>
   rotation=<rot 90 0 0>
  }]
 dyn_skirt_mid_back_01_l [{
   shapeType=Box
   extentX=15
   extentY=2
   extentZ=15
   shapeName="None, None, None, None, None, None, None, None, None, None, None, None, None, None, None, None, None, None, None, None, None, None, None, None, None, None, None, None, None, None, None, None, None, None"
   center=<7.5 0 0>
   rotation=<rot 90 0 0>
  }]
 dyn_skirt_mid_back_01_r [{
   shapeType=Box
   extentX=15
   extentY=2
   extentZ=15
   shapeName="None, None, None, None, None, None, None, None, None, None, None, None, None, None, None, None, None, None, None, None, None, None, None, None, None, None, None, None, None, None, None, None, None, None"
   center=<7.5 0 0>
   rotation=<rot 90 0 0>
  }]
 dyn_skirt_mid_back_02_l [{
   shapeType=Box
   extentX=15
   extentY=2
   extentZ=15
   shapeName="None, None, None, None, None, None, None, None, None, None, None, None, None, None, None, None, None, None, None, None, None, None, None, None, None, None, None, None, None, None, None, None, None, None"
   center=<7.5 0 0>
   rotation=<rot 90 0 0>
  }]
 dyn_skirt_mid_back_02_r [{
   shapeType=Box
   extentX=15
   extentY=2
   extentZ=15
   shapeName="None, None, None, None, None, None, None, None, None, None, None, None, None, None, None, None, None, None, None, None, None, None, None, None, None, None, None, None, None, None, None, None, None, None"
   center=<7.5 0 0>
   rotation=<rot 90 0 0>
  }]
 dyn_skirt_mid_back_03_l [{
   shapeType=Box
   extentX=15
   extentY=2
   extentZ=15
   shapeName="None, None, None, None, None, None, None, None, None, None, None, None, None, None, None, None, None, None, None, None, None, None, None, None, None, None, None, None, None, None, None, None, None, None"
   center=<7.5 0 0>
   rotation=<rot 90 0 0>
  }]
 dyn_skirt_mid_back_03_r [{
   shapeType=Box
   extentX=15
   extentY=2
   extentZ=15
   shapeName="None, None, None, None, None, None, None, None, None, None, None, None, None, None, None, None, None, None, None, None, None, None, None, None, None, None, None, None, None, None, None, None, None, None"
   center=<7.5 0 0>
   rotation=<rot 90 0 0>
  }]
 dyn_skirt_mid_back_04_l [{
   shapeType=Box
   extentX=15
   extentY=2
   extentZ=15
   shapeName="None, None, None, None, None, None, None, None, None, None, None, None, None, None, None, None, None, None, None, None, None, None, None, None, None, None, None, None, None, None, None, None, None, None"
   center=<7.5 0 0>
   rotation=<rot 90 0 0>
  }]
 dyn_skirt_mid_back_04_r [{
   shapeType=Box
   extentX=15
   extentY=2
   extentZ=15
   shapeName="None, None, None, None, None, None, None, None, None, None, None, None, None, None, None, None, None, None, None, None, None, None, None, None, None, None, None, None, None, None, None, None, None, None"
   center=<7.5 0 0>
   rotation=<rot 90 0 0>
  }]
 dyn_skirt_mid_back_05_l [{
   shapeType=Box
   extentX=15
   extentY=2
   extentZ=15
   shapeName="None, None, None, None, None, None, None, None, None, None, None, None, None, None, None, None, None, None, None, None, None, None, None, None, None, None, None, None, None, None, None, None, None, None"
   center=<7.5 0 0>
   rotation=<rot 90 0 0>
  }]
 dyn_skirt_mid_back_05_r [{
   shapeType=Box
   extentX=15
   extentY=2
   extentZ=15
   shapeName="None, None, None, None, None, None, None, None, None, None, None, None, None, None, None, None, None, None, None, None, None, None, None, None, None, None, None, None, None, None, None, None, None, None"
   center=<7.5 0 0>
   rotation=<rot 90 0 0>
  }]
 dyn_skirt_mid_front_01_l [{
   shapeType=Box
   extentX=15
   extentY=2
   extentZ=15
   shapeName="None, None, None, None, None, None, None, None, None, None, None, None, None, None, None, None, None, None, None, None, None, None, None, None, None, None, None, None, None, None, None, None, None, None"
   center=<7.5 0 0>
   rotation=<rot 90 0 0>
  }]
 dyn_skirt_mid_front_01_r [{
   shapeType=Box
   extentX=15
   extentY=2
   extentZ=15
   shapeName="None, None, None, None, None, None, None, None, None, None, None, None, None, None, None, None, None, None, None, None, None, None, None, None, None, None, None, None, None, None, None, None, None, None"
   center=<7.5 0 0>
   rotation=<rot 90 0 0>
  }]
 dyn_skirt_mid_front_02_l [{
   shapeType=Box
   extentX=15
   extentY=2
   extentZ=15
   shapeName="None, None, None, None, None, None, None, None, None, None, None, None, None, None, None, None, None, None, None, None, None, None, None, None, None, None, None, None, None, None, None, None, None, None"
   center=<7.5 0 0>
   rotation=<rot 90 0 0>
  }]
 dyn_skirt_mid_front_02_r [{
   shapeType=Box
   extentX=15
   extentY=2
   extentZ=15
   shapeName="None, None, None, None, None, None, None, None, None, None, None, None, None, None, None, None, None, None, None, None, None, None, None, None, None, None, None, None, None, None, None, None, None, None"
   center=<7.5 0 0>
   rotation=<rot 90 0 0>
  }]
 dyn_skirt_mid_front_03_l [{shapeType=Box, extentX=15, extentY=2, extentZ=15, center=<7.5 0 0>, rotation=<rot 90 0 0>}]
 dyn_skirt_mid_front_03_r [{
   shapeType=Box
   extentX=15
   extentY=2
   extentZ=28
   shapeName="None, None, None, None, None, None, None, None, None, None, None, None, None, None, None, None, None, None, None, None, None, None, None, None, None, None, None, None, None, None, None, None, None, None"
   center=<14.3725 0 0.0004>
   rotation=<rot 90 0 0>
  }]
 dyn_skirt_mid_front_04_l [{shapeType=Box, extentX=15, extentY=2, extentZ=15, center=<7.5 0 0>, rotation=<rot 90 0 0>}]
 dyn_skirt_side_02_l [{
   shapeType=Box
   extentX=15
   extentY=2
   extentZ=15
   shapeName="None, None, None, None, None, None, None, None, None, None, None, None, None, None, None, None, None, None, None, None, None, None, None, None, None, None, None, None, None, None, None, None, None, None"
   center=<7.5 0 0>
   rotation=<rot 90 0 0>
  }]
 dyn_skirt_side_02_r [{
   shapeType=Box
   extentX=15
   extentY=2
   extentZ=15
   shapeName="None, None, None, None, None, None, None, None, None, None, None, None, None, None, None, None, None, None, None, None, None, None, None, None, None, None, None, None, None, None, None, None, None, None"
   center=<7.5 0 0>
   rotation=<rot 90 0 0>
  }]
 dyn_skirt_side_03_l [{
   shapeType=Box
   extentX=15
   extentY=2
   extentZ=15
   shapeName="None, None, None, None, None, None, None, None, None, None, None, None, None, None, None, None, None, None, None, None, None, None, None, None, None, None, None, None, None, None, None, None, None, None"
   center=<7.5 0 0>
   rotation=<rot 90 0 0>
  }]
 dyn_skirt_side_03_r [{
   shapeType=Box
   extentX=15
   extentY=2
   extentZ=15
   shapeName="None, None, None, None, None, None, None, None, None, None, None, None, None, None, None, None, None, None, None, None, None, None, None, None, None, None, None, None, None, None, None, None, None, None"
   center=<7.5 0 0>
   rotation=<rot 90 0 0>
  }]
 dyn_skirt_side_04_l [{shapeType=Box, extentX=15, extentY=2, extentZ=15, center=<7.5 0 0>, rotation=<rot 90 0 0>}]
 dyn_skirt_side_04_r [{
   shapeType=Box
   extentX=15
   extentY=2
   extentZ=25
   shapeName="None, None, None, None, None, None, None, None, None, None, None, None, None, None, None, None, None, None, None, None, None, None, None, None, None, None, None, None, None, None, None, None, None, None"
   center=<12.8252 0 -0.0003>
   rotation=<rot 90 0 0>
  }]
 dyn_skirt_side_05_l [{shapeType=Box, extentX=15, extentY=2, extentZ=15, center=<7.5 0 0>, rotation=<rot 90 0 0>}]
 foot_l [{shapeType=Capsule, length=11.1544, radius=6.0379, center=<-2.8982 3.378 0.6775>, rotation=<rot 0 20.1476 -80.761>}]
 foot_r [{shapeType=Capsule, length=11.1775, radius=6.0623, center=<2.9255 -3.3876 -0.7016>, rotation=<rot 0 20.6335 -80.5537>}]
 pelvis [{shapeType=Capsule, length=10.5, radius=10, center=<-0.1765 0.7647 0>}]
 thigh_l [{shapeType=Capsule, length=40.25, radius=8.25, center=<-24.7796 0.0004 0>, rotation=<rot 90 0 0>}]
 thigh_r [{shapeType=Capsule, length=39.5, radius=8.25, center=<23.6681 0.0003 0>, rotation=<rot 90 0 0>}]
}
constraints [
 {bone1Name=dyn_skirt_side_03_l, bone2Name=dyn_skirt_side_02_l, swing1LimitDegrees=30, swing2LimitDegrees=30, twistLimitDegrees=0}
 {bone1Name=dyn_skirt_mid_03_l, bone2Name=dyn_skirt_mid_02_l, swing1LimitDegrees=30, swing2LimitDegrees=30, twistLimitDegrees=0}
 {bone1Name=dyn_skirt_mid_04_l, bone2Name=dyn_skirt_mid_03_l, swing1LimitDegrees=30, swing2LimitDegrees=30, twistLimitDegrees=0}
 {bone1Name=dyn_skirt_mid_05_l, bone2Name=dyn_skirt_mid_04_l, swing1LimitDegrees=30, swing2LimitDegrees=30, twistLimitDegrees=0}
 {bone1Name=dyn_skirt_back_03, bone2Name=dyn_skirt_back_02, swing1LimitDegrees=30, swing2LimitDegrees=30, twistLimitDegrees=0}
 {bone1Name=dyn_skirt_back_04, bone2Name=dyn_skirt_back_03, swing1LimitDegrees=30, swing2LimitDegrees=30, twistLimitDegrees=0}
 {bone1Name=dyn_skirt_back_05, bone2Name=dyn_skirt_back_04, swing1LimitDegrees=30, swing2LimitDegrees=30, twistLimitDegrees=0}
 {bone1Name=dyn_skirt_back_06, bone2Name=dyn_skirt_back_05, swing1LimitDegrees=30, swing2LimitDegrees=30, twistLimitDegrees=0}
 {bone1Name=dyn_skirt_back_07, bone2Name=dyn_skirt_back_06, swing1LimitDegrees=30, swing2LimitDegrees=30, twistLimitDegrees=0}
 {bone1Name=dyn_skirt_mid_03_r, bone2Name=dyn_skirt_mid_02_r, swing1LimitDegrees=30, swing2LimitDegrees=30, twistLimitDegrees=0}
 {bone1Name=dyn_skirt_mid_04_r, bone2Name=dyn_skirt_mid_03_r, swing1LimitDegrees=30, swing2LimitDegrees=30, twistLimitDegrees=0}
 {bone1Name=dyn_skirt_mid_05_r, bone2Name=dyn_skirt_mid_04_r, swing1LimitDegrees=30, swing2LimitDegrees=30, twistLimitDegrees=0}
 {bone1Name=dyn_skirt_side_03_r, bone2Name=dyn_skirt_side_02_r, swing1LimitDegrees=30, swing2LimitDegrees=30, twistLimitDegrees=0}
 {bone1Name=dyn_skirt_mid_back_02_r, bone2Name=dyn_skirt_mid_back_01_r, swing1LimitDegrees=30, swing2LimitDegrees=30, twistLimitDegrees=0}
 {bone1Name=dyn_skirt_mid_back_03_r, bone2Name=dyn_skirt_mid_back_02_r, swing1LimitDegrees=30, swing2LimitDegrees=30, twistLimitDegrees=0}
 {bone1Name=dyn_skirt_mid_back_04_r, bone2Name=dyn_skirt_mid_back_03_r, swing1LimitDegrees=30, swing2LimitDegrees=30, twistLimitDegrees=0}
 {bone1Name=dyn_skirt_mid_back_05_r, bone2Name=dyn_skirt_mid_back_04_r, swing1LimitDegrees=30, swing2LimitDegrees=30, twistLimitDegrees=0}
 {bone1Name=dyn_skirt_mid_front_02_r, bone2Name=dyn_skirt_mid_front_01_r, swing1LimitDegrees=30, swing2LimitDegrees=30, twistLimitDegrees=0}
 {bone1Name=dyn_skirt_mid_back_02_l, bone2Name=dyn_skirt_mid_back_01_l, swing1LimitDegrees=30, swing2LimitDegrees=30, twistLimitDegrees=0}
 {bone1Name=dyn_skirt_mid_back_03_l, bone2Name=dyn_skirt_mid_back_02_l, swing1LimitDegrees=30, swing2LimitDegrees=30, twistLimitDegrees=0}
 {bone1Name=dyn_skirt_mid_back_04_l, bone2Name=dyn_skirt_mid_back_03_l, swing1LimitDegrees=30, swing2LimitDegrees=30, twistLimitDegrees=0}
 {bone1Name=dyn_skirt_mid_back_05_l, bone2Name=dyn_skirt_mid_back_04_l, swing1LimitDegrees=30, swing2LimitDegrees=30, twistLimitDegrees=0}
 {bone1Name=dyn_skirt_mid_front_02_l, bone2Name=dyn_skirt_mid_front_01_l, swing1LimitDegrees=30, swing2LimitDegrees=30, twistLimitDegrees=0}
 {bone1Name=dyn_skirt_side_02_l, bone2Name=pelvis, swing1LimitDegrees=30, swing2LimitDegrees=30, twistLimitDegrees=0}
 {bone1Name=dyn_skirt_mid_02_l, bone2Name=pelvis, swing1LimitDegrees=30, swing2LimitDegrees=30, twistLimitDegrees=0}
 {bone1Name=dyn_skirt_back_02, bone2Name=pelvis, swing1LimitDegrees=30, swing2LimitDegrees=30, twistLimitDegrees=0}
 {bone1Name=dyn_skirt_mid_02_r, bone2Name=pelvis, swing1LimitDegrees=30, swing2LimitDegrees=30, twistLimitDegrees=0}
 {bone1Name=dyn_skirt_side_02_r, bone2Name=pelvis, swing1LimitDegrees=30, swing2LimitDegrees=30, twistLimitDegrees=0}
 {bone1Name=dyn_skirt_mid_back_01_r, bone2Name=pelvis, swing1LimitDegrees=30, swing2LimitDegrees=30, twistLimitDegrees=0}
 {bone1Name=dyn_skirt_mid_front_01_r, bone2Name=pelvis, swing1LimitDegrees=30, swing2LimitDegrees=30, twistLimitDegrees=0}
 {bone1Name=dyn_skirt_mid_back_01_l, bone2Name=pelvis, swing1LimitDegrees=30, swing2LimitDegrees=30, twistLimitDegrees=0}
 {bone1Name=dyn_skirt_mid_front_01_l, bone2Name=pelvis, swing1LimitDegrees=30, swing2LimitDegrees=30, twistLimitDegrees=0}
 {bone1Name=thigh_l, bone2Name=pelvis, swing1LimitDegrees=60, swing2LimitDegrees=60, twistLimitDegrees=5}
 {bone1Name=calf_l, bone2Name=thigh_l, swing1LimitDegrees=60, swing2LimitDegrees=60, twistLimitDegrees=5}
 {bone1Name=thigh_r, bone2Name=pelvis, swing1LimitDegrees=60, swing2LimitDegrees=60, twistLimitDegrees=5}
 {bone1Name=calf_r, bone2Name=thigh_r, swing1LimitDegrees=60, swing2LimitDegrees=60, twistLimitDegrees=5}
 {bone1Name=foot_l, bone2Name=calf_l, swing1LimitDegrees=60, swing2LimitDegrees=60, twistLimitDegrees=5}
 {bone1Name=foot_r, bone2Name=calf_r, swing1LimitDegrees=60, swing2LimitDegrees=60, twistLimitDegrees=5}
 {
  bone1Name=dyn_skirt_mid_front_02_r
  bone2Name=dyn_skirt_side_03_r
  swing1LimitDegrees=45
  swing1Motion=Free
  swing2LimitDegrees=45
  swing2Motion=Free
  twistLimitDegrees=45
  twistMotion=Free
 }
 {
  bone1Name=dyn_skirt_back_07
  bone2Name=dyn_skirt_mid_back_05_r
  swing1LimitDegrees=45
  swing1Motion=Free
  swing2LimitDegrees=45
  swing2Motion=Free
  twistLimitDegrees=45
  twistMotion=Free
 }
 {
  bone1Name=dyn_skirt_back_07
  bone2Name=dyn_skirt_mid_back_05_l
  swing1LimitDegrees=45
  swing1Motion=Free
  swing2LimitDegrees=45
  swing2Motion=Free
  twistLimitDegrees=45
  twistMotion=Free
 }
 {
  bone1Name=dyn_skirt_back_06
  bone2Name=dyn_skirt_mid_back_04_l
  swing1LimitDegrees=45
  swing1Motion=Free
  swing2LimitDegrees=45
  swing2Motion=Free
  twistLimitDegrees=45
  twistMotion=Free
 }
 {
  bone1Name=dyn_skirt_back_06
  bone2Name=dyn_skirt_mid_back_04_r
  swing1LimitDegrees=45
  swing1Motion=Free
  swing2LimitDegrees=45
  swing2Motion=Free
  twistLimitDegrees=45
  twistMotion=Free
 }
 {bone1Name=dyn_skirt_side_04_r, bone2Name=dyn_skirt_side_03_r, swing1LimitDegrees=30, swing2LimitDegrees=30, twistLimitDegrees=0}
 {bone1Name=dyn_skirt_mid_front_03_r, bone2Name=dyn_skirt_mid_front_02_r, swing1LimitDegrees=30, swing2LimitDegrees=30, twistLimitDegrees=0}
 {
  bone1Name=dyn_skirt_mid_front_02_l
  bone2Name=dyn_skirt_side_03_l
  swing1LimitDegrees=45
  swing1Motion=Free
  swing2LimitDegrees=45
  swing2Motion=Free
  twistLimitDegrees=45
  twistMotion=Free
 }
 {
  bone1Name=dyn_skirt_mid_back_04_r
  bone2Name=dyn_skirt_mid_05_r
  swing1LimitDegrees=45
  swing1Motion=Free
  swing2LimitDegrees=45
  swing2Motion=Free
  twistLimitDegrees=45
  twistMotion=Free
 }
 {
  bone1Name=dyn_skirt_mid_04_r
  bone2Name=dyn_skirt_mid_front_03_r
  swing1LimitDegrees=45
  swing1Motion=Free
  swing2LimitDegrees=45
  swing2Motion=Free
  twistLimitDegrees=45
  twistMotion=Free
 }
 {bone1Name=dyn_skirt_side_04_l, bone2Name=dyn_skirt_side_03_l, swing1LimitDegrees=30, swing2LimitDegrees=30, twistLimitDegrees=0}
 {bone1Name=dyn_skirt_side_05_l, bone2Name=dyn_skirt_side_04_l, swing1LimitDegrees=30, swing2LimitDegrees=30, twistLimitDegrees=0}
 {bone1Name=dyn_skirt_mid_front_03_l, bone2Name=dyn_skirt_mid_front_02_l, swing1LimitDegrees=30, swing2LimitDegrees=30, twistLimitDegrees=0}
 {bone1Name=dyn_skirt_mid_front_04_l, bone2Name=dyn_skirt_mid_front_03_l, swing1LimitDegrees=30, swing2LimitDegrees=30, twistLimitDegrees=0}
 {
  bone1Name=dyn_skirt_mid_front_04_l
  bone2Name=dyn_skirt_side_05_l
  swing1LimitDegrees=45
  swing1Motion=Free
  swing2LimitDegrees=45
  swing2Motion=Free
  twistLimitDegrees=45
  twistMotion=Free
 }
 {
  bone1Name=dyn_skirt_mid_front_03_r
  bone2Name=dyn_skirt_mid_05_r
  swing1LimitDegrees=45
  swing1Motion=Free
  swing2LimitDegrees=45
  swing2Motion=Free
  twistLimitDegrees=45
  twistMotion=Free
 }
 {
  bone1Name=dyn_skirt_mid_front_03_r
  bone2Name=dyn_skirt_side_04_r
  swing1LimitDegrees=45
  swing1Motion=Free
  swing2LimitDegrees=45
  swing2Motion=Free
  twistLimitDegrees=45
  twistMotion=Free
 }
]
