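# Post-process-style cosmetic AnimBP for the Ice Queen body, copying the parent mesh pose and adding secondary motion to skirt, leg cloth and fur.
# Full simulation chains RigidBody and AnimDynamics, with separate emote and normal tunings; the normal branch adds parachute skirt constraints and gravity override.
# Above LOD 1 or with rigid bodies disabled, bone constraints, springs and trail controllers replace simulation.
# EventGraph caches those switches, parachute state and gravity.
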AnimBlueprint /BRCosmetics/Characters/Player/Female/Medium/Bodies/F_MED_IceQueen/Meshes/F_MED_Ice_Queen_Body_AnimBP
target_skeleton=&/F_MED_Ice_Queen_Body_Skeleton
graphs [
 AnimGraph {nodes [
   n1 = Anim Dynamics - Bone: dyn_leg_cloth_04_r(Alpha=1, ExternalForce="0,0,0") {
    angularDampingOverride=0.9999
    angularSpringConstant=1500
    bAngularSpring=true
    bGravityOverrideInSimSpace=true
    bOverrideAngularDamping=true
    gravityScale=0.35
    simulationSpace=RootRelative
    alphaBoolBlend {blendOption=Linear}
    boundBone {boneName=dyn_leg_cloth_04_r}
    componentAppliedLinearAccClamp=<100 100 100>
    physicsBodyDefinitions [{
      sphereCollisionRadius=0
      boundBone {boneName=dyn_leg_cloth_04_r}
      boxExtents=<10 20 20>
      constraintSetup {angularLimitsMax=<0 10 30>, angularLimitsMin=<0 -10 -30>, angularTarget=<0 0 0>}
      localJointOffset=<10 0 0>
     }]
   }
   n2 = Anim Dynamics - Bone: dyn_leg_cloth_04_l(ComponentPose=@n1, Alpha=1, ExternalForce="0,0,0") {
    angularDampingOverride=0.9999
    angularSpringConstant=1500
    bAngularSpring=true
    bGravityOverrideInSimSpace=true
    bOverrideAngularDamping=true
    gravityScale=0.35
    simulationSpace=RootRelative
    alphaBoolBlend {blendOption=Linear}
    boundBone {boneName=dyn_leg_cloth_04_l}
    componentAppliedLinearAccClamp=<100 100 100>
    physicsBodyDefinitions [{
      sphereCollisionRadius=0
      boundBone {boneName=dyn_leg_cloth_04_l}
      boxExtents=<10 20 20>
      constraintSetup {angularLimitsMax=<0 10 30>, angularLimitsMin=<0 -10 -30>, angularTarget=<0 0 0>}
      localJointOffset=<10 0 0>
     }]
   }
   n3 = Anim Dynamics - Bone: dyn_leg_cloth_05_r(ComponentPose=@n2, Alpha=1, ExternalForce="0,0,0") {
    angularDampingOverride=0.9999
    angularSpringConstant=1500
    bAngularSpring=true
    bGravityOverrideInSimSpace=true
    bOverrideAngularDamping=true
    gravityScale=0.35
    simulationSpace=RootRelative
    alphaBoolBlend {blendOption=Linear}
    boundBone {boneName=dyn_leg_cloth_05_r}
    componentAppliedLinearAccClamp=<100 100 100>
    physicsBodyDefinitions [{
      sphereCollisionRadius=0
      boundBone {boneName=dyn_leg_cloth_05_r}
      boxExtents=<10 20 20>
      constraintSetup {angularLimitsMax=<0 10 30>, angularLimitsMin=<0 -10 -30>, angularTarget=<0 0 0>}
      localJointOffset=<10 0 0>
     }]
   }
   Anim Dynamics - Bone: dyn_leg_cloth_05_l(ComponentPose=@n3, Alpha=1, ExternalForce="0,0,0") {
    angularDampingOverride=0.9999
    angularSpringConstant=1500
    bAngularSpring=true
    bGravityOverrideInSimSpace=true
    bOverrideAngularDamping=true
    gravityScale=0.35
    simulationSpace=RootRelative
    alphaBoolBlend {blendOption=Linear}
    boundBone {boneName=dyn_leg_cloth_05_l}
    componentAppliedLinearAccClamp=<100 100 100>
    physicsBodyDefinitions [{
      sphereCollisionRadius=0
      boundBone {boneName=dyn_leg_cloth_05_l}
      boxExtents=<10 20 20>
      constraintSetup {angularLimitsMax=<0 10 30>, angularLimitsMin=<0 -10 -30>, angularTarget=<0 0 0>}
      localJointOffset=<10 0 0>
     }]
   }
   n4 = "Constraint - dyn_leg_cloth_01_r (P)"(Alpha=0.3, ConstraintWeights_0=1) {
    alphaBoolBlend {blendOption=Linear}
    boneToModify {boneName=dyn_leg_cloth_01_r}
    constraintSetup [{targetBone {boneName=thigh_r}}]
    constraintWeights [1]
   }
   n5 = "Constraint - dyn_leg_cloth_01_l (P)"(ComponentPose=@n4, Alpha=0.3, ConstraintWeights_0=1) {
    alphaBoolBlend {blendOption=Linear}
    boneToModify {boneName=dyn_leg_cloth_01_l}
    constraintSetup [{targetBone {boneName=thigh_l}}]
    constraintWeights [1]
   }
   n6 = "Constraint - dyn_leg_cloth_02_r (P)"(ComponentPose=@n5, Alpha=0.1, ConstraintWeights_0=1) {
    alphaBoolBlend {blendOption=Linear}
    boneToModify {boneName=dyn_leg_cloth_02_r}
    constraintSetup [{targetBone {boneName=thigh_r}}]
    constraintWeights [1]
   }
   "Constraint - dyn_leg_cloth_02_l (P)"(ComponentPose=@n6, Alpha=0.1, ConstraintWeights_0=1) {
    alphaBoolBlend {blendOption=Linear}
    boneToModify {boneName=dyn_leg_cloth_02_l}
    constraintSetup [{targetBone {boneName=thigh_l}}]
    constraintWeights [1]
   }
   n7 = "Constraint - dyn_leg_cloth_03_r (P)"(Alpha=0.05, ConstraintWeights_0=1) {
    alphaBoolBlend {blendOption=Linear}
    boneToModify {boneName=dyn_leg_cloth_03_r}
    constraintSetup [{targetBone {boneName=thigh_r}}]
    constraintWeights [1]
   }
   "Constraint - dyn_leg_cloth_03_l (P)"(ComponentPose=@n7, Alpha=0.05, ConstraintWeights_0=1) {
    alphaBoolBlend {blendOption=Linear}
    boneToModify {boneName=dyn_leg_cloth_03_l}
    constraintSetup [{targetBone {boneName=thigh_l}}]
    constraintWeights [1]
   }
   n8 = Get parachute_open() -> (parachute_open:double)
   n9 = "Constraint - dyn_skirt_back_02 (P)"(Alpha=@n8, ConstraintWeights_0=1) {
    alphaBoolBlend {blendOption=Linear}
    boneToModify {boneName=dyn_skirt_back_02}
    constraintSetup [{targetBone {boneName=pelvis}}]
    constraintWeights [1]
   }
   n10 = "Constraint - dyn_skirt_mid_02_l (P)"(ComponentPose=@n9, Alpha=@n8, ConstraintWeights_0=1) {
    alphaBoolBlend {blendOption=Linear}
    boneToModify {boneName=dyn_skirt_mid_02_l}
    constraintSetup [{targetBone {boneName=pelvis}}]
    constraintWeights [1]
   }
   n11 = "Constraint - dyn_skirt_mid_02_r (P)"(ComponentPose=@n10, Alpha=@n8, ConstraintWeights_0=1) {
    alphaBoolBlend {blendOption=Linear}
    boneToModify {boneName=dyn_skirt_mid_02_r}
    constraintSetup [{targetBone {boneName=pelvis}}]
    constraintWeights [1]
   }
   n12 = "Transform (Modify) Bone - Bone: dyn_skirt_back_02"(ComponentPose=@n11, Alpha=@n8, Translation="0,-1,0", Rotation="0,0,0", Scale="1,1,1") {
    rotationMode=BMM_Additive
    rotationSpace=BCS_BoneSpace
    scaleMode=BMM_Additive
    scaleSpace=BCS_BoneSpace
    translationMode=BMM_Additive
    translationSpace=BCS_BoneSpace
    alphaBoolBlend {blendOption=Linear}
    boneToModify {boneName=dyn_skirt_back_02}
    translation {y=-1}
   }
   n13 = "Transform (Modify) Bone - Bone: dyn_skirt_mid_02_r"(ComponentPose=@n12, Alpha=@n8, Translation="0,-1,0", Rotation="0,0,0", Scale="1,1,1") {
    rotationMode=BMM_Additive
    rotationSpace=BCS_BoneSpace
    scaleMode=BMM_Additive
    scaleSpace=BCS_BoneSpace
    translationMode=BMM_Additive
    translationSpace=BCS_BoneSpace
    alphaBoolBlend {blendOption=Linear}
    boneToModify {boneName=dyn_skirt_mid_02_r}
    translation {y=-1}
   }
   "Transform (Modify) Bone - Bone: dyn_skirt_mid_02_l"(ComponentPose=@n13, Alpha=@n8, Translation="0,-1,0", Rotation="0,0,0", Scale="1,1,1") {
    rotationMode=BMM_Additive
    rotationSpace=BCS_BoneSpace
    scaleMode=BMM_Additive
    scaleSpace=BCS_BoneSpace
    translationMode=BMM_Additive
    translationSpace=BCS_BoneSpace
    alphaBoolBlend {blendOption=Linear}
    boneToModify {boneName=dyn_skirt_mid_02_l}
    translation {y=-1}
   }
   n14 = "Constraint - dyn_skirt_side_02_l (PP)"!(ComponentPose=Local To Component(LocalPose=Use cached pose 'CopyPoseCache'()), Alpha=1, ConstraintWeights_0=1, ConstraintWeights_1:float=0) -> (Pose) {
    alphaBoolBlend {blendOption=Linear}
    boneToModify {boneName=dyn_skirt_side_02_l}
    constraintSetup [{targetBone {boneName=thigh_l}}, {targetBone {boneName=pelvis}}]
    constraintWeights [1,1]
   }
   n15 = "Constraint - dyn_skirt_side_02_r (PP)"!(ComponentPose=@n14, Alpha=1, ConstraintWeights_0=1, ConstraintWeights_1:float=0) -> (Pose) {
    alphaBoolBlend {blendOption=Linear}
    boneToModify {boneName=dyn_skirt_side_02_r}
    constraintSetup [{targetBone {boneName=thigh_r}}, {targetBone {boneName=pelvis}}]
    constraintWeights [1,1]
   }
   n16 = Break Fort Anim Input Skydiving External Force(FortAnimInput_SkydivingExternalForce=Get SkydivingExternalForceInput()) -> (
    bUseSkydivingVectorForce=~
    SkydivingVectorForce=~
    HeadToPelvisDirection=~
    FloatingMultiplier=~
    FloatingAdditive=~
    DivingMultiplier=~
    DivingAdditive=~
    ParachutingMultiplier=~
    ParachutingAdditive=~
   )
   n17 = RigidBody(ComponentPose=@n15, ExternalForce=vector * float(A=@n16.SkydivingVectorForce, B=0.4)) {
    bOverrideWorldGravity=true
    overridePhysicsAsset=&/F_MED_Ice_Queen_Body_Physics
    alphaBoolBlend {blendOption=Linear}
    baseBoneRef {boneName=root}
    componentAppliedLinearAccClamp=<100 100 100>
    componentLinearAccScale=<1 1 1>
    componentLinearVelScale=<1 1 1>
    overrideWorldGravity {z=-900}
   }
   n18 = Anim Dynamics - Bone: dyn_leg_cloth_01_r(ComponentPose=@n17, Alpha=1, ExternalForce="0,0,0") {
    angularDampingOverride=1
    angularSpringConstant=1500
    bAngularSpring=true
    bGravityOverrideInSimSpace=true
    bOverrideAngularDamping=true
    simulationSpace=RootRelative
    alphaBoolBlend {blendOption=Linear}
    boundBone {boneName=dyn_leg_cloth_01_r}
    componentAppliedLinearAccClamp=<1000 1000 1000>
    componentLinearAccScale=<1 1 1>
    componentLinearVelScale=<1 1 1>
    physicsBodyDefinitions [{
      sphereCollisionRadius=0
      boundBone {boneName=dyn_leg_cloth_01_r}
      boxExtents=<10 20 20>
      constraintSetup {angularLimitsMax=<0 0 20>, angularLimitsMin=<0 0 0>, angularTarget=<0 0 0>}
      localJointOffset=<10 0 0>
     }]
   }
   n19 = Anim Dynamics - Bone: dyn_leg_cloth_01_l(ComponentPose=@n18, Alpha=1, ExternalForce="0,0,0") {
    angularDampingOverride=1
    angularSpringConstant=1500
    bAngularSpring=true
    bGravityOverrideInSimSpace=true
    bOverrideAngularDamping=true
    simulationSpace=RootRelative
    alphaBoolBlend {blendOption=Linear}
    boundBone {boneName=dyn_leg_cloth_01_l}
    componentLinearAccScale=<5 5 5>
    componentLinearVelScale=<5 5 5>
    physicsBodyDefinitions [{
      sphereCollisionRadius=0
      boundBone {boneName=dyn_leg_cloth_01_l}
      boxExtents=<10 20 20>
      constraintSetup {angularLimitsMax=<0 0 20>, angularLimitsMin=<0 0 0>, angularTarget=<0 0 0>}
      localJointOffset=<10 0 0>
     }]
   }
   n20 = Anim Dynamics - Bone: dyn_leg_cloth_02_r(ComponentPose=@n19, Alpha=1, ExternalForce="0,0,0") {
    angularDampingOverride=1
    angularSpringConstant=1500
    bAngularSpring=true
    bGravityOverrideInSimSpace=true
    bOverrideAngularDamping=true
    simulationSpace=RootRelative
    alphaBoolBlend {blendOption=Linear}
    boundBone {boneName=dyn_leg_cloth_02_r}
    componentAppliedLinearAccClamp=<1000 1000 1000>
    componentLinearAccScale=<1 1 1>
    componentLinearVelScale=<1 1 1>
    physicsBodyDefinitions [{
      sphereCollisionRadius=0
      boundBone {boneName=dyn_leg_cloth_02_r}
      boxExtents=<10 20 20>
      constraintSetup {angularLimitsMax=<0 10 20>, angularLimitsMin=<0 -10 0>, angularTarget=<0 0 0>}
      localJointOffset=<10 0 0>
     }]
   }
   n21 = Anim Dynamics - Bone: dyn_leg_cloth_02_l(ComponentPose=@n20, Alpha=1, ExternalForce="0,0,0") {
    angularDampingOverride=1
    angularSpringConstant=1500
    bAngularSpring=true
    bGravityOverrideInSimSpace=true
    bOverrideAngularDamping=true
    simulationSpace=RootRelative
    alphaBoolBlend {blendOption=Linear}
    boundBone {boneName=dyn_leg_cloth_02_l}
    componentAppliedLinearAccClamp=<1000 1000 1000>
    componentLinearAccScale=<1 1 1>
    componentLinearVelScale=<1 1 1>
    physicsBodyDefinitions [{
      sphereCollisionRadius=0
      boundBone {boneName=dyn_leg_cloth_02_l}
      boxExtents=<10 20 20>
      constraintSetup {angularLimitsMax=<0 10 20>, angularLimitsMin=<0 -10 0>, angularTarget=<0 0 0>}
      localJointOffset=<10 0 0>
     }]
   }
   n22 = Anim Dynamics - Bone: dyn_leg_cloth_03_r(ComponentPose=@n21, Alpha=1, ExternalForce="0,0,0") {
    angularDampingOverride=1
    angularSpringConstant=1500
    bAngularSpring=true
    bGravityOverrideInSimSpace=true
    bOverrideAngularDamping=true
    simulationSpace=RootRelative
    alphaBoolBlend {blendOption=Linear}
    boundBone {boneName=dyn_leg_cloth_03_r}
    componentAppliedLinearAccClamp=<1000 1000 1000>
    componentLinearAccScale=<1 1 1>
    componentLinearVelScale=<1 1 1>
    physicsBodyDefinitions [{
      sphereCollisionRadius=0
      boundBone {boneName=dyn_leg_cloth_03_r}
      boxExtents=<10 20 20>
      constraintSetup {angularLimitsMax=<0 10 30>, angularLimitsMin=<0 -10 -20>, angularTarget=<0 0 0>}
      localJointOffset=<10 0 0>
     }]
   }
   n23 = Anim Dynamics - Bone: dyn_leg_cloth_03_l(ComponentPose=@n22, Alpha=1, ExternalForce="0,0,0") {
    angularDampingOverride=1
    angularSpringConstant=1500
    bAngularSpring=true
    bGravityOverrideInSimSpace=true
    bOverrideAngularDamping=true
    simulationSpace=RootRelative
    alphaBoolBlend {blendOption=Linear}
    boundBone {boneName=dyn_leg_cloth_03_l}
    componentAppliedLinearAccClamp=<1000 1000 1000>
    componentLinearAccScale=<1 1 1>
    componentLinearVelScale=<1 1 1>
    physicsBodyDefinitions [{
      sphereCollisionRadius=0
      boundBone {boneName=dyn_leg_cloth_03_l}
      boxExtents=<10 20 20>
      constraintSetup {angularLimitsMax=<0 10 30>, angularLimitsMin=<0 -10 -20>, angularTarget=<0 0 0>}
      localJointOffset=<10 0 0>
     }]
   }
   n24 = RigidBody(ComponentPose=@n23) {
    bOverrideWorldGravity=true
    overridePhysicsAsset=&/F_MED_Ice_Queen_Body_Bits_Physics
    simulationSpace=BaseBoneSpace
    alphaBoolBlend {blendOption=Linear}
    baseBoneRef {boneName=pelvis}
    overrideWorldGravity {z=-980}
   }
   n25 = Component To Local(ComponentPose=@n24)
   n26 = "Constraint - dyn_skirt_side_02_l (PP)"(ComponentPose=Local To Component(LocalPose=Use cached pose 'CopyPoseCache'()), Alpha=1, ConstraintWeights_0=1, ConstraintWeights_1=0) {
    alphaBoolBlend {blendOption=Linear}
    boneToModify {boneName=dyn_skirt_side_02_l}
    constraintSetup [{targetBone {boneName=thigh_l}}, {targetBone {boneName=pelvis}}]
    constraintWeights [1,1]
   }
   n27 = "Constraint - dyn_skirt_side_02_r (PP)"(ComponentPose=@n26, Alpha=1, ConstraintWeights_0=1, ConstraintWeights_1=0) {
    alphaBoolBlend {blendOption=Linear}
    boneToModify {boneName=dyn_skirt_side_02_r}
    constraintSetup [{targetBone {boneName=thigh_r}}, {targetBone {boneName=pelvis}}]
    constraintWeights [1,1]
   }
   n28 = Get parachute_open()
   n29 = "Constraint - dyn_skirt_back_02 (P)"(ComponentPose=@n27, Alpha=@n28, ConstraintWeights_0=1) {
    alphaBoolBlend {blendOption=Linear}
    boneToModify {boneName=dyn_skirt_back_02}
    constraintSetup [{targetBone {boneName=pelvis}}]
    constraintWeights [1]
   }
   n30 = "Constraint - dyn_skirt_mid_02_l (P)"(ComponentPose=@n29, Alpha=@n28, ConstraintWeights_0=1) {
    alphaBoolBlend {blendOption=Linear}
    boneToModify {boneName=dyn_skirt_mid_02_l}
    constraintSetup [{targetBone {boneName=pelvis}}]
    constraintWeights [1]
   }
   n31 = "Constraint - dyn_skirt_mid_02_r (P)"(ComponentPose=@n30, Alpha=@n28, ConstraintWeights_0=1) {
    alphaBoolBlend {blendOption=Linear}
    boneToModify {boneName=dyn_skirt_mid_02_r}
    constraintSetup [{targetBone {boneName=pelvis}}]
    constraintWeights [1]
   }
   n32 = RigidBody!(ComponentPose=@n31, Alpha, OverrideWorldGravity:Vector=Get Gravity Override(), ExternalForce) -> (Pose) {
    bOverrideWorldGravity=true
    overridePhysicsAsset=&F_MED_Ice_Queen_Body_Physics
    alphaBoolBlend {blendOption=Linear}
    baseBoneRef {boneName=root}
    componentAppliedLinearAccClamp=<6000 6000 6000>
    componentLinearAccScale=<0.05 0.05 0.05>
    componentLinearVelScale=<0.05 0.05 0.05>
    overrideWorldGravity {z=-900}
   }
   n33 = Anim Dynamics - Bone: dyn_leg_cloth_01_r(ComponentPose=@n32, Alpha=1, ExternalForce="0,0,0") {
    angularDampingOverride=1
    angularSpringConstant=1500
    bAngularSpring=true
    bGravityOverrideInSimSpace=true
    bOverrideAngularDamping=true
    gravityScale=0.35
    simulationSpace=RootRelative
    alphaBoolBlend {blendOption=Linear}
    boundBone {boneName=dyn_leg_cloth_01_r}
    componentAppliedLinearAccClamp=<100 100 100>
    physicsBodyDefinitions [{
      sphereCollisionRadius=0
      boundBone {boneName=dyn_leg_cloth_01_r}
      boxExtents=<10 20 20>
      constraintSetup {angularLimitsMax=<0 0 20>, angularLimitsMin=<0 0 0>, angularTarget=<0 0 0>}
      localJointOffset=<10 0 0>
     }]
   }
   n34 = Anim Dynamics - Bone: dyn_leg_cloth_01_l(ComponentPose=@n33, Alpha=1, ExternalForce="0,0,0") {
    angularDampingOverride=1
    angularSpringConstant=1500
    bAngularSpring=true
    bGravityOverrideInSimSpace=true
    bOverrideAngularDamping=true
    gravityScale=0.35
    simulationSpace=RootRelative
    alphaBoolBlend {blendOption=Linear}
    boundBone {boneName=dyn_leg_cloth_01_l}
    componentAppliedLinearAccClamp=<100 100 100>
    physicsBodyDefinitions [{
      sphereCollisionRadius=0
      boundBone {boneName=dyn_leg_cloth_01_l}
      boxExtents=<10 20 20>
      constraintSetup {angularLimitsMax=<0 0 20>, angularLimitsMin=<0 0 0>, angularTarget=<0 0 0>}
      localJointOffset=<10 0 0>
     }]
   }
   n35 = Anim Dynamics - Bone: dyn_leg_cloth_02_r(ComponentPose=@n34, Alpha=1, ExternalForce="0,0,0") {
    angularDampingOverride=1
    angularSpringConstant=1500
    bAngularSpring=true
    bGravityOverrideInSimSpace=true
    bOverrideAngularDamping=true
    gravityScale=0.35
    simulationSpace=RootRelative
    alphaBoolBlend {blendOption=Linear}
    boundBone {boneName=dyn_leg_cloth_02_r}
    componentAppliedLinearAccClamp=<100 100 100>
    physicsBodyDefinitions [{
      sphereCollisionRadius=0
      boundBone {boneName=dyn_leg_cloth_02_r}
      boxExtents=<10 20 20>
      constraintSetup {angularLimitsMax=<0 10 20>, angularLimitsMin=<0 -10 0>, angularTarget=<0 0 0>}
      localJointOffset=<10 0 0>
     }]
   }
   n36 = Anim Dynamics - Bone: dyn_leg_cloth_02_l(ComponentPose=@n35, Alpha=1, ExternalForce="0,0,0") {
    angularDampingOverride=1
    angularSpringConstant=1500
    bAngularSpring=true
    bGravityOverrideInSimSpace=true
    bOverrideAngularDamping=true
    gravityScale=0.35
    simulationSpace=RootRelative
    alphaBoolBlend {blendOption=Linear}
    boundBone {boneName=dyn_leg_cloth_02_l}
    componentAppliedLinearAccClamp=<100 100 100>
    physicsBodyDefinitions [{
      sphereCollisionRadius=0
      boundBone {boneName=dyn_leg_cloth_02_l}
      boxExtents=<10 20 20>
      constraintSetup {angularLimitsMax=<0 10 20>, angularLimitsMin=<0 -10 0>, angularTarget=<0 0 0>}
      localJointOffset=<10 0 0>
     }]
   }
   n37 = Anim Dynamics - Bone: dyn_leg_cloth_03_r(ComponentPose=@n36, Alpha=1, ExternalForce="0,0,0") {
    angularDampingOverride=1
    angularSpringConstant=1500
    bAngularSpring=true
    bGravityOverrideInSimSpace=true
    bOverrideAngularDamping=true
    gravityScale=0.35
    simulationSpace=RootRelative
    alphaBoolBlend {blendOption=Linear}
    boundBone {boneName=dyn_leg_cloth_03_r}
    componentAppliedLinearAccClamp=<100 100 100>
    physicsBodyDefinitions [{
      sphereCollisionRadius=0
      boundBone {boneName=dyn_leg_cloth_03_r}
      boxExtents=<10 20 20>
      constraintSetup {angularLimitsMax=<0 10 30>, angularLimitsMin=<0 -10 -20>, angularTarget=<0 0 0>}
      localJointOffset=<10 0 0>
     }]
   }
   n38 = Anim Dynamics - Bone: dyn_leg_cloth_03_l(ComponentPose=@n37, Alpha=1, ExternalForce="0,0,0") {
    angularDampingOverride=1
    angularSpringConstant=1500
    bAngularSpring=true
    bGravityOverrideInSimSpace=true
    bOverrideAngularDamping=true
    gravityScale=0.35
    simulationSpace=RootRelative
    alphaBoolBlend {blendOption=Linear}
    boundBone {boneName=dyn_leg_cloth_03_l}
    componentAppliedLinearAccClamp=<100 100 100>
    physicsBodyDefinitions [{
      sphereCollisionRadius=0
      boundBone {boneName=dyn_leg_cloth_03_l}
      boxExtents=<10 20 20>
      constraintSetup {angularLimitsMax=<0 10 30>, angularLimitsMin=<0 -10 -20>, angularTarget=<0 0 0>}
      localJointOffset=<10 0 0>
     }]
   }
   n39 = Break Fort Anim Input Skydiving External Force(FortAnimInput_SkydivingExternalForce=Get SkydivingExternalForceInput()) -> (
    bUseSkydivingVectorForce=~
    SkydivingVectorForce=~
    HeadToPelvisDirection=~
    FloatingMultiplier=~
    FloatingAdditive=~
    DivingMultiplier=~
    DivingAdditive=~
    ParachutingMultiplier=~
    ParachutingAdditive=~
   )
   n40 = RigidBody!(ComponentPose=@n38, Alpha, ExternalForce=vector * float(A=@n39.SkydivingVectorForce, B=0.1)) -> (Pose) {
    bOverrideWorldGravity=true
    overridePhysicsAsset=&F_MED_Ice_Queen_Body_Bits_Physics
    simulationSpace=BaseBoneSpace
    alphaBoolBlend {blendOption=Linear}
    baseBoneRef {boneName=pelvis}
    overrideWorldGravity {z=-980}
   }
   n41 = Component To Local(ComponentPose=@n40)
   n42 = Get bIsPlayingEmote()
   n43 = "Constraint - dyn_skirt_side_02_l (PP)"(ComponentPose=Local To Component(LocalPose=Use cached pose 'CopyPoseCache'()), Alpha=1, ConstraintWeights_0=1, ConstraintWeights_1=0) {
    alphaBoolBlend {blendOption=Linear}
    boneToModify {boneName=dyn_skirt_side_02_l}
    constraintSetup [{targetBone {boneName=thigh_l}}, {targetBone {boneName=pelvis}}]
    constraintWeights [1,1]
   }
   n44 = "Constraint - dyn_skirt_side_02_r (PP)"(ComponentPose=@n43, Alpha=1, ConstraintWeights_0=1, ConstraintWeights_1=0) {
    alphaBoolBlend {blendOption=Linear}
    boneToModify {boneName=dyn_skirt_side_02_r}
    constraintSetup [{targetBone {boneName=thigh_r}}, {targetBone {boneName=pelvis}}]
    constraintWeights [1,1]
   }
   n45 = "Constraint - dyn_leg_cloth_04_r (P)"(ComponentPose=@n44, Alpha=1, ConstraintWeights_0=1) {
    alphaBoolBlend {blendOption=Linear}
    boneToModify {boneName=dyn_leg_cloth_04_r}
    constraintSetup [{targetBone {boneName=calf_r}}]
    constraintWeights [1]
   }
   n46 = "Constraint - dyn_leg_cloth_04_l (P)"(ComponentPose=@n45, Alpha=1, ConstraintWeights_0=1) {
    alphaBoolBlend {blendOption=Linear}
    boneToModify {boneName=dyn_leg_cloth_04_l}
    constraintSetup [{targetBone {boneName=calf_l}}]
    constraintWeights [1]
   }
   n47 = Spring controller - Bone: dyn_fur_04_r!(
    ComponentPose:Component Space Pose Link Structure=Spring controller - Bone: dyn_fur_01_r!(
     ComponentPose:Component Space Pose Link Structure=Spring controller - Bone: dyn_fur_02_r!(ComponentPose:Component Space Pose Link Structure=@n46, Alpha:float=1) -> (Pose:Component Space Pose Link Structure) {
      is=title AnimGraphNode_SpringBone
      bLimitDisplacement=true
      maxDisplacement=1
      alphaBoolBlend {blendOption=Linear}
      springBone {boneName=dyn_fur_02_r}
     }
     Alpha:float=1
    ) -> (Pose:Component Space Pose Link Structure) {
     is=title AnimGraphNode_SpringBone
     bLimitDisplacement=true
     maxDisplacement=1
     alphaBoolBlend {blendOption=Linear}
     springBone {boneName=dyn_fur_01_r}
    }
    Alpha:float=1
   ) -> (Pose:Component Space Pose Link Structure) {
    is=title AnimGraphNode_SpringBone
    bLimitDisplacement=true
    maxDisplacement=1
    alphaBoolBlend {blendOption=Linear}
    springBone {boneName=dyn_fur_04_r}
   }
   n48 = Spring controller - Bone: dyn_fur_02_l!(
    ComponentPose:Component Space Pose Link Structure=Spring controller - Bone: dyn_fur_01_l!(
     ComponentPose:Component Space Pose Link Structure=Spring controller - Bone: dyn_fur_03_r!(ComponentPose:Component Space Pose Link Structure=@n47, Alpha:float=1) -> (Pose:Component Space Pose Link Structure) {
      is=title AnimGraphNode_SpringBone
      bLimitDisplacement=true
      maxDisplacement=1
      alphaBoolBlend {blendOption=Linear}
      springBone {boneName=dyn_fur_03_r}
     }
     Alpha:float=1
    ) -> (Pose:Component Space Pose Link Structure) {
     is=title AnimGraphNode_SpringBone
     bLimitDisplacement=true
     maxDisplacement=1
     alphaBoolBlend {blendOption=Linear}
     springBone {boneName=dyn_fur_01_l}
    }
    Alpha:float=1
   ) -> (Pose:Component Space Pose Link Structure) {
    is=title AnimGraphNode_SpringBone
    bLimitDisplacement=true
    maxDisplacement=1
    alphaBoolBlend {blendOption=Linear}
    springBone {boneName=dyn_fur_02_l}
   }
   n49 = Trail controller - Bone: dyn_skirt_back_05(
    ComponentPose=Spring controller - Bone: dyn_fur_04_l!(
     ComponentPose:Component Space Pose Link Structure=Spring controller - Bone: dyn_fur_03_l!(ComponentPose:Component Space Pose Link Structure=@n48, Alpha:float=1) -> (Pose:Component Space Pose Link Structure) {
      is=title AnimGraphNode_SpringBone
      bLimitDisplacement=true
      maxDisplacement=1
      alphaBoolBlend {blendOption=Linear}
      springBone {boneName=dyn_fur_03_l}
     }
     Alpha:float=1
    ) -> (Pose:Component Space Pose Link Structure) {
     is=title AnimGraphNode_SpringBone
     bLimitDisplacement=true
     maxDisplacement=1
     alphaBoolBlend {blendOption=Linear}
     springBone {boneName=dyn_fur_04_l}
    }
    Alpha=0.5
   ) {
    bLimitRotation=true
    bUsePlanarLimit=true
    chainLength=4
    relaxationSpeedScale=2.5
    planarLimits [{drivingBone {boneName=dyn_skirt_back_02}, planeTransform {location=<0 2 0>, rotation=<rot 0 5 -90>, scale=<1 1 1>}}]
    rotationLimits [
     {limitMax=<30 30 30>, limitMin=<-30 -30 -30>}
     {limitMax=<30 30 30>, limitMin=<-30 -30 -30>}
     {limitMax=<30 30 30>, limitMin=<-30 -30 -30>}
     {limitMax=<30 30 30>, limitMin=<-30 -30 -30>}
    ]
    rotationOffsets [<0 0 0>, <0 0 0>, <0 0 0>, <0 0 0>]
    trailBone {boneName=dyn_skirt_back_05}
   }
   n50 = Trail controller - Bone: dyn_skirt_mid_06_r(ComponentPose=@n49, Alpha=0.5) {
    bLimitRotation=true
    bUsePlanarLimit=true
    chainLength=4
    relaxationSpeedScale=2.5
    planarLimits [{drivingBone {boneName=dyn_skirt_back_02}, planeTransform {location=<0 2 0>, rotation=<rot 0 5.0001 -89.9999>, scale=<1 1 1>}}]
    rotationLimits [
     {limitMax=<30 30 30>, limitMin=<-30 -30 -30>}
     {limitMax=<30 30 30>, limitMin=<-30 -30 -30>}
     {limitMax=<30 30 30>, limitMin=<-30 -30 -30>}
     {limitMax=<30 30 30>, limitMin=<-30 -30 -30>}
    ]
    rotationOffsets [<0 0 0>, <0 0 0>, <0 0 0>, <0 0 0>]
    trailBone {boneName=dyn_skirt_mid_06_r}
   }
   n51 = Trail controller - Bone: dyn_skirt_mid_back_05_r(ComponentPose=@n50, Alpha=0.5) {
    bLimitRotation=true
    bUsePlanarLimit=true
    chainLength=4
    relaxationSpeedScale=2.5
    planarLimits [{drivingBone {boneName=dyn_skirt_back_02}, planeTransform {location=<0 2 0>, rotation=<rot 0 5.0001 -89.9999>, scale=<1 1 1>}}]
    rotationLimits [
     {limitMax=<30 30 30>, limitMin=<-30 -30 -30>}
     {limitMax=<30 30 30>, limitMin=<-30 -30 -30>}
     {limitMax=<30 30 30>, limitMin=<-30 -30 -30>}
     {limitMax=<30 30 30>, limitMin=<-30 -30 -30>}
    ]
    rotationOffsets [<0 0 0>, <0 0 0>, <0 0 0>, <0 0 0>]
    trailBone {boneName=dyn_skirt_mid_back_05_r}
   }
   n52 = Trail controller - Bone: dyn_skirt_mid_06_l(ComponentPose=@n51, Alpha=0.5) {
    bLimitRotation=true
    bUsePlanarLimit=true
    chainLength=4
    relaxationSpeedScale=2.5
    planarLimits [{drivingBone {boneName=dyn_skirt_back_02}, planeTransform {location=<0 2 0>, rotation=<rot 0 5.0001 -89.9999>, scale=<1 1 1>}}]
    rotationLimits [
     {limitMax=<30 30 30>, limitMin=<-30 -30 -30>}
     {limitMax=<30 30 30>, limitMin=<-30 -30 -30>}
     {limitMax=<30 30 30>, limitMin=<-30 -30 -30>}
     {limitMax=<30 30 30>, limitMin=<-30 -30 -30>}
    ]
    rotationOffsets [<0 0 0>, <0 0 0>, <0 0 0>, <0 0 0>]
    trailBone {boneName=dyn_skirt_mid_06_l}
   }
   n53 = Trail controller - Bone: dyn_skirt_mid_back_05_l(ComponentPose=@n52, Alpha=0.5) {
    bLimitRotation=true
    bUsePlanarLimit=true
    chainLength=4
    relaxationSpeedScale=2.5
    planarLimits [{drivingBone {boneName=dyn_skirt_back_02}, planeTransform {location=<0 2 0>, rotation=<rot 0 5.0001 -89.9999>, scale=<1 1 1>}}]
    rotationLimits [
     {limitMax=<30 30 30>, limitMin=<-30 -30 -30>}
     {limitMax=<30 30 30>, limitMin=<-30 -30 -30>}
     {limitMax=<30 30 30>, limitMin=<-30 -30 -30>}
     {limitMax=<30 30 30>, limitMin=<-30 -30 -30>}
    ]
    rotationOffsets [<0 0 0>, <0 0 0>, <0 0 0>, <0 0 0>]
    trailBone {boneName=dyn_skirt_mid_back_05_l}
   }
   n54 = "integer > integer"(A=Get CurrentLODAnim() -> (CurrentLODAnim:int=0), B=1) {is=title K2Node_CallFunction}
   n55 = NOT Boolean(A=Get RigidBodyEnabledAnim() -> (RigidBodyEnabledAnim:bool))
   Output Pose(Result=Blend Poses by bool(
     BlendPose_0=Component To Local(ComponentPose=@n53)
     BlendPose_1=Blend Poses by bool(BlendPose_0=@n25, BlendPose_1=@n41, BlendTime_0=0.3, BlendTime_1=0.3, bActiveValue=@n42) {blendType=Linear}
     bActiveValue=OR Boolean(A=@n55, B=@n54)
    ) {blendType=Linear}) {layerGroup=None}
   Save cached pose 'CopyPoseCache'(Pose=Copy Pose From Mesh() {bUseAttachedParent=true}) {cacheName=CopyPoseCache}
   Trail controller - Bone: dyn_skirt_back_05(Alpha=1) {
    bUsePlanarLimit=true
    chainLength=4
    relaxationSpeedScale=2
    planarLimits [{drivingBone {boneName=dyn_skirt_back_02}, planeTransform {location=<0 2 0>, rotation=<rot 0 5.0001 -89.9999>, scale=<1 1 1>}}]
    rotationLimits [{limitMax=<180 180 180>}, {limitMax=<180 180 180>}, {limitMax=<180 180 180>}, {limitMax=<180 180 180>}]
    rotationOffsets [<0 0 0>, <0 0 0>, <0 0 0>, <0 0 0>]
    trailBone {boneName=dyn_skirt_back_05}
   }
   Use cached pose 'CopyPoseCache'()
   vector * float(A=@n16.SkydivingVectorForce, B=0.5)
   vector * float(A=@n39.SkydivingVectorForce, B=0.5)
   Get CurrentLOD()
   Get bRigidBodyEnabled()
  ]}
 EventGraph {nodes [
   Event BlueprintUpdateAnimation()
   > n1 = Sequence()
   > Set RigidBodyEnabledAnim(RigidBodyEnabledAnim:bool=Get bRigidBodyEnabled()) -> (Output_Get:bool)
   > Set CurrentLODAnim(CurrentLODAnim:int=Get CurrentLOD()) -> (Output_Get:int=0)
   > n2 = Branch(Condition=Get bIsParachuteOpened())
   > Set parachute_open(parachute_open:double=1) -> (Output_Get:double)
   Set parachute_open(execute=@n2.else, parachute_open=0)
   n3 = Gravity Override State Function DEPRECATED(execute=@n1.then_1, CharacterPartAnimInstance=Self-Reference(), GameTimeInSeconds=0, GravityOverrideParams=Get Gravity Override Params())
   > Set Gravity Override(Gravity Override=@n3)
  ]}
]
variables [
 CurrentLODAnim {type=integer}
 Gravity Override {type=Vector}
 Gravity Override Params {type=GravityOverrideParamsStruct}
 parachute_open {type=number}
 RigidBodyEnabledAnim {type=boolean}
]
