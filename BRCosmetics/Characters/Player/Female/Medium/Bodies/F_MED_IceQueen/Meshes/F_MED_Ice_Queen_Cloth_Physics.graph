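# Physics asset for the female medium Ice Queen body cosmetic, named for cloth physics and covering only the lower body.
# Single capsules approximate the pelvis and both thighs, calves and feet, forming simple collision volumes.
# Constraints chain each foot to its calf, calf to thigh, and thigh to pelvis, with all swing and twist motion left free.
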PhysicsAsset /BRCosmetics/Characters/Player/Female/Medium/Bodies/F_MED_IceQueen/Meshes/F_MED_Ice_Queen_Cloth_Physics
bodies {
 calf_l [{shapeType=Capsule, length=32.9896, radius=7.5121, center=<-22.5899 -2.5139 0.0002>, rotation=<rot 90 0 176.1264>}]
 calf_r [{shapeType=Capsule, length=32.9896, radius=7.5121, center=<22.5899 2.5139 0.0002>, rotation=<rot 90 0 -3.8736>}]
 foot_l [{shapeType=Capsule, length=12.5194, radius=7.5379, center=<-2.8982 3.378 0.6775>, rotation=<rot 0 20.1476 -80.7611>}]
 foot_r [{shapeType=Capsule, length=12.5194, radius=7.5622, center=<2.9255 -3.3876 -0.7016>, rotation=<rot 0 20.6336 -80.5539>}]
 pelvis [{shapeType=Capsule, length=13.9398, radius=13.0606, center=<-7.5753 -3.5711 -0.5654>, rotation=<rot -3.412 -90 0.7142>}]
 thigh_l [{shapeType=Capsule, length=32.3256, radius=8, center=<-25.5414 0.0001 0>, rotation=<rot 90 0 175.2225>}]
 thigh_r [{shapeType=Capsule, length=32.3256, radius=8, center=<25.5414 -0.0001 0>, rotation=<rot 90 0 -4.7775>}]
}
constraints [
 {
  bone1Name=calf_r
  bone2Name=thigh_r
  swing1LimitDegrees=45
  swing1Motion=Free
  swing2LimitDegrees=45
  swing2Motion=Free
  twistLimitDegrees=45
  twistMotion=Free
 }
 {
  bone1Name=foot_l
  bone2Name=calf_l
  swing1LimitDegrees=45
  swing1Motion=Free
  swing2LimitDegrees=45
  swing2Motion=Free
  twistLimitDegrees=45
  twistMotion=Free
 }
 {
  bone1Name=foot_r
  bone2Name=calf_r
  swing1LimitDegrees=45
  swing1Motion=Free
  swing2LimitDegrees=45
  swing2Motion=Free
  twistLimitDegrees=45
  twistMotion=Free
 }
 {
  bone1Name=thigh_r
  bone2Name=pelvis
  swing1LimitDegrees=45
  swing1Motion=Free
  swing2LimitDegrees=45
  swing2Motion=Free
  twistLimitDegrees=45
  twistMotion=Free
 }
 {
  bone1Name=thigh_l
  bone2Name=pelvis
  swing1LimitDegrees=45
  swing1Motion=Free
  swing2LimitDegrees=45
  swing2Motion=Free
  twistLimitDegrees=45
  twistMotion=Free
 }
 {
  bone1Name=calf_l
  bone2Name=thigh_l
  swing1LimitDegrees=45
  swing1Motion=Free
  swing2LimitDegrees=45
  swing2Motion=Free
  twistLimitDegrees=45
  twistMotion=Free
 }
]
